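; Battle Pass S19 gem fragment quest ability that refreshes vehicle-collect fragments overlapping the player.
; On activation from an event, it calls the parent and switches on the event tag for vehicle entered or exit.
; After a short delay, it gets overlapping BP_S19_GemFragment_Collect_Vehicle actors from the owning pawn and calls UpdateOverlappingActors on each.
; The ability then ends.
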(event Ability|EventActivateAbilityFromEvent
  (params (EventData :type Gameplay%20Event%20Data%20Structure%20%28by%20ref%29))
  (bind _owningplayerpawn (Variables|GAVars|Data|GetOwningPlayerPawn))
  (bind _overlappingactors (Collision|GetOverlappingActors _owningplayerpawn "/BattlePassS19/QuestActors/GemFragments/BP_S19_GemFragment_Collect_Vehicle.BP_S19_GemFragment_Collect_Vehicle_C" :!class /Script/Engine.Actor))
  (|CallParentFunction EventData)
  (switch Utilities|FlowControl|Switch|SwitchonGameplayTag (Utilities|Struct|BreakGameplayEventData EventData)
    (:Event.Vehicle.Entered
      _)
    (:Event.Vehicle.Exit
      _))
  (Utilities|FlowControl|Delay 1.0)
  (if (Utilities|Array|IsValidIndex _overlappingactors)
    (for _array_element _overlappingactors
      (bind _asbp_s19_gem_fragment_collect_vehicle (Utilities|Casting|CastToBP_S19_GemFragment_Collect_Vehicle _array_element))
      (Class|BPAthenaPropQuestActorParent|UpdateOverlappingActors _asbp_s19_gem_fragment_collect_vehicle))
    (else
      _))
  (Ability|EndAbility))
